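# Character master material for the Fire skin of the male medium Cereal Box body, supporting default lit and subsurface profile shading.
# Between shared PreFX and PostFX functions, mask-driven static switches chain optional panning emissive, generic gem and sequin effects.
# The panning emissive is modulated by scrolling grayscale fire noise, projected from pre-skinned local position with a front/top-down blend, then scaled by time of day.
# A cinematic post-FX function finishes the chain.
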Material /BRCosmetics/Characters/Player/Male/Medium/Bodies/M_MED_Cereal_Box/Skins/Fire/Materials/M_MED_CerealBox_Fire_Master
graph {
 properties {
  bUseMaterialAttributes=true
  bUsedWithMeshDeformer=true
  bUsedWithMorphTargets=true
  bUsedWithSkeletalMesh=true
  pixelDepthOffsetMode=PDOM_Legacy
  shadingModel=MSM_FromMaterialExpression
  subsurfaceProfile=&SS_HeroSkin_02
  usedShadingModels=Default Lit | Subsurface Profile
 }
 nodes [
  n1 = MaterialFunctionCall() {materialFunction=&BaseFN_CharacterFunction_PreFX}
  n2 = TextureObjectParameter() {group=Skin FX, parameterName=SkinFX_Mask, samplerType=SAMPLERTYPE_Masks, texture=&T_WhiteMask}
  n3 = TextureSample(Tex=@n2) {samplerType=SAMPLERTYPE_Masks}
  n4 = StaticComponentMaskParameter(None=@n3.RGBA) {defaultR=true, group=Skin FX, parameterName=PanningEmissive_Channel}
  n5 = GetMaterialAttributes(None=MaterialFunctionCall!(
    Alpha=@n4
    isVehicle
    Gmap
    UVOffset=StaticComponentMaskParameter(None=@n3.RGBA) {defaultB=true, group=Skin FX, parameterName=PanningEmissive_UVOffset_Channel}
    Local Normal
    Local Position
    EmissiveColorOverride
    Material=@n1.Result
   ) {materialFunction=/Game/Weapons/WeaponSkins/MaterialFunction/PanningEmissive}.Result) {attributeGetTypes [#1]}
  n6 = ScalarParameter() {group=CerealBoxFire, parameterName=CB_FireMin}
  n7 = ScalarParameter() {defaultValue=10, group=CerealBoxFire, parameterName=CB_FireMax}
  n8 = LinearInterpolate(
   A=ComponentMask(None=@n1.Pre skinned Local Position) {b=true, r=true}
   B=ComponentMask(None=@n1.Pre skinned Local Position) {g=true, r=true}
   Alpha=ScalarParameter() {group=CerealBoxFire, parameterName=CB_FireTopDownUV}
  )
  n9 = ScalarParameter() {defaultValue=0.01, group=CerealBoxFire, parameterName=CB_FireTiling}
  n10 = Time()
  n11 = AppendVector(
   A=ScalarParameter() {defaultValue=1, group=CerealBoxFire, parameterName=CB_FireSpeedX}
   B=ScalarParameter() {defaultValue=1, group=CerealBoxFire, parameterName=CB_FireSpeedZ}
  )
  n12 = TextureSampleParameter2D(UVs=Add(A=Multiply(A=@n8, B=@n9), B=Multiply(A=@n10, B=@n11))) {
   group=CerealBoxFire
   parameterName=CB_FireTex
   samplerType=SAMPLERTYPE_LinearGrayscale
   texture=/BRCosmetics/Weapons/WeaponSkins/Wraps/Textures/GlowRandom/T_Wrap_GlowRandom_M
  }
  n13 = Multiply(
   A=Multiply(A=@n5.Emissive Color, B=LinearInterpolate(A=@n6, B=@n7, Alpha=@n12.R))
   B=MaterialFunctionCall(
    Value_Morning=Constant() {r=0.15}
    Value_Day=Constant() {r=1}
    Value_Evening=Constant() {r=0.25}
    Value_Night=ScalarParameter() {defaultValue=0.55}
   ) {materialFunction=&Scalar_Time_of_Day}.Result
  )
  n14 = StaticSwitchParameter(
   True=SetMaterialAttributes(MaterialAttributes=@n5.MaterialAttributes, Emissive Color=LinearInterpolate(A=@n5.Emissive Color, B=@n13, Alpha=@n4)) {attributeSetTypes [#1]}
   False=@n1.Result
  ) {group=Skin FX, parameterName=Use Panning Emissive}
  n15 = StaticSwitchParameter(
   True=MaterialFunctionCall!(
    MASK=StaticComponentMaskParameter(None=@n3.RGBA) {defaultR=true, group=Skin FX, parameterName=Generic Gem Effect Channel}
    Color Change Sequence=StaticComponentMaskParameter(None=@n3.RGBA) {defaultB=true, group=Skin FX, parameterName=Generic Gem Color Change Channel}
    Material=@n14
   ) {materialFunction=/Game/Weapons/WeaponSkins/MaterialFunction/GenericGem}.Result
   False=@n14
  ) {group=Skin FX, parameterName=Use Generic Gem}
  n16 = MaterialFunctionCall(
   AO=@n1.AO
   Distance Blend=@n1.AO
   Fresnel=@n1.AO
   Gmap=@n1.AO
   Pre Skinned Local Position=@n1.AO
   Material=StaticSwitchParameter(
    True=MaterialFunctionCall!(Mask=@n2, TechArtMask, inMat=@n15) {materialFunction=/Game/Weapons/WeaponSkins/MaterialFunction/MF_Sequins}.Result
    False=@n15
   ) {group=Skin FX, parameterName=UseSequins}
  ) {materialFunction=&BaseFN_CharacterFunction_PostFX}
  MaterialFunctionCall(In=@n16.Result) {materialFunction=&CINE_MF_CharPostFX}
 ]
}
